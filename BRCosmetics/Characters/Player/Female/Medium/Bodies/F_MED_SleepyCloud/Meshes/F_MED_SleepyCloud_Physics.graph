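# Physics asset for the F_MED_SleepyCloud female medium player body, covering the pants drawstring cords.
# A pelvis capsule and a small base_pantscords capsule anchor two chains of small capsules, four links on the left and three on the right.
# The base joins the pelvis with free swing and twist, while each cord link constrains to its parent with narrow swing and twist limit angles.
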PhysicsAsset /BRCosmetics/Characters/Player/Female/Medium/Bodies/F_MED_SleepyCloud/Meshes/F_MED_SleepyCloud_Physics
bodies {
 base_pantscords [{shapeType=Capsule, length=0.505, radius=0.505, shapeName=base_pantscords_capsule, rotation=<rot 0 -90 0>}]
 dyn_pantcord_l_01 [{shapeType=Capsule, length=1, radius=1, center=<1 0 0>, rotation=<rot 90 0 0>}]
 dyn_pantcord_l_02 [{shapeType=Capsule, length=1, radius=1, center=<1 0 0>, rotation=<rot 90 0 0>}]
 dyn_pantcord_l_03 [{shapeType=Capsule, length=1, radius=1, center=<1 0 0>, rotation=<rot 90 0 0>}]
 dyn_pantcord_l_04 [{shapeType=Capsule, length=1, radius=1, center=<1 0 0>, rotation=<rot 90 0 0>}]
 dyn_pantcord_r_01 [{shapeType=Capsule, length=1, radius=1, center=<1 0 0>, rotation=<rot 90 0 0>}]
 dyn_pantcord_r_02 [{shapeType=Capsule, length=1, radius=1, center=<1 0 0>, rotation=<rot 90 0 0>}]
 dyn_pantcord_r_03 [{shapeType=Capsule, length=1, radius=1, center=<1 0 0>, rotation=<rot 90 0 0>}]
 pelvis [{shapeType=Capsule, length=10, radius=12, shapeName=pelvis_capsule, center=<1.6988 -3.6549 0>}]
}
constraints [
 {
  bone1Name=base_pantscords
  bone2Name=pelvis
  swing1LimitDegrees=45
  swing1Motion=Free
  swing2LimitDegrees=45
  swing2Motion=Free
  twistLimitDegrees=45
  twistMotion=Free
 }
 {bone1Name=dyn_pantcord_l_01, bone2Name=base_pantscords, swing1LimitDegrees=5, swing2LimitDegrees=10, twistLimitDegrees=2}
 {bone1Name=dyn_pantcord_l_02, bone2Name=dyn_pantcord_l_01, swing1LimitDegrees=5, swing2LimitDegrees=10, twistLimitDegrees=2}
 {bone1Name=dyn_pantcord_l_03, bone2Name=dyn_pantcord_l_02, swing1LimitDegrees=5, swing2LimitDegrees=10, twistLimitDegrees=2}
 {bone1Name=dyn_pantcord_l_04, bone2Name=dyn_pantcord_l_03, swing1LimitDegrees=5, swing2LimitDegrees=10, twistLimitDegrees=2}
 {bone1Name=dyn_pantcord_r_01, bone2Name=base_pantscords, swing1LimitDegrees=5, swing2LimitDegrees=10, twistLimitDegrees=2}
 {bone1Name=dyn_pantcord_r_02, bone2Name=dyn_pantcord_r_01, swing1LimitDegrees=5, swing2LimitDegrees=10, twistLimitDegrees=2}
 {bone1Name=dyn_pantcord_r_03, bone2Name=dyn_pantcord_r_02, swing1LimitDegrees=5, swing2LimitDegrees=10, twistLimitDegrees=2}
]
