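# Physics asset for the face-accessory part of the female medium SleepyCloud cosmetic body, covering its ponytail and front-hair bones.
# Capsules on upper spine, neck and head form the body, with both hair base bones constrained to the head with free rotation.
# A four-link ponytail chain swings mainly on one axis with slight twist, loosening toward the tip.
# A three-link front-hair chain uses tighter swing and twist limits.
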PhysicsAsset /BRCosmetics/Characters/Player/Female/Medium/Bodies/F_MED_SleepyCloud/Meshes/Parts/F_MED_SleepyCloud_FaceAcc_BP_Physics
bodies {
 base_PonyTail [{shapeType=Capsule, length=0.505, radius=0.505, shapeName=base_PonyTail_capsule, rotation=<rot 0 -90 0>}]
 base_frontHair [{shapeType=Capsule, length=0.505, radius=0.505, shapeName=base_frontHair_capsule, rotation=<rot 0 -90 0>}]
 dyn_PonyTail_01 [{shapeType=Capsule, length=2, radius=2, center=<1 0 0>, rotation=<rot 90 0 0>}]
 dyn_PonyTail_02 [{shapeType=Capsule, length=2, radius=2, center=<1.4133 0 0>, rotation=<rot 90 0 0>}]
 dyn_PonyTail_03 [{shapeType=Capsule, length=2, radius=2, center=<1 0 0>, rotation=<rot 90 0 0>}]
 dyn_PonyTail_04 [{shapeType=Capsule, length=2, radius=2, center=<3 0 0>, rotation=<rot 90 0 0>}]
 dyn_frontHair_01 [{shapeType=Capsule, length=1, radius=0.5, center=<0.1 0 0>, rotation=<rot 90 0 0>}]
 dyn_frontHair_02 [{shapeType=Capsule, length=1, radius=0.5, center=<0.1 0 0>, rotation=<rot 90 0 0>}]
 dyn_frontHair_03 [{shapeType=Capsule, length=1, radius=0.5, center=<1 0 0>, rotation=<rot 90 0 0>}]
 head [{shapeType=Capsule, length=8, radius=5, shapeName=head_capsule, center=<3.3568 0 0>, rotation=<rot 90 9.9999 0>}]
 neck_01 [{shapeType=Capsule, length=5.5, radius=5.5, shapeName=neck_01_capsule, center=<3.6928 0.4058 0>, rotation=<rot 90 0 0>}]
 spine_05 [{shapeType=Capsule, length=20.5, radius=8, shapeName=spine_05_capsule, center=<5.3212 -0.6974 0>}]
}
constraints [
 {
  bone1Name=neck_01
  bone2Name=spine_05
  swing1LimitDegrees=45
  swing1Motion=Free
  swing2LimitDegrees=45
  swing2Motion=Free
  twistLimitDegrees=45
  twistMotion=Free
 }
 {
  bone1Name=head
  bone2Name=neck_01
  swing1LimitDegrees=45
  swing1Motion=Free
  swing2LimitDegrees=45
  swing2Motion=Free
  twistLimitDegrees=45
  twistMotion=Free
 }
 {
  bone1Name=base_PonyTail
  bone2Name=head
  swing1LimitDegrees=45
  swing1Motion=Free
  swing2LimitDegrees=45
  swing2Motion=Free
  twistLimitDegrees=45
  twistMotion=Free
 }
 {bone1Name=dyn_PonyTail_01, bone2Name=base_PonyTail, swing1LimitDegrees=20, swing2LimitDegrees=0, twistLimitDegrees=3}
 {bone1Name=dyn_PonyTail_02, bone2Name=dyn_PonyTail_01, swing1LimitDegrees=20, swing2LimitDegrees=0, twistLimitDegrees=3}
 {bone1Name=dyn_PonyTail_03, bone2Name=dyn_PonyTail_02, swing1LimitDegrees=30, swing2LimitDegrees=0, twistLimitDegrees=3}
 {bone1Name=dyn_PonyTail_04, bone2Name=dyn_PonyTail_03, swing1LimitDegrees=30, swing2LimitDegrees=0, twistLimitDegrees=3}
 {
  bone1Name=base_frontHair
  bone2Name=head
  swing1LimitDegrees=45
  swing1Motion=Free
  swing2LimitDegrees=45
  swing2Motion=Free
  twistLimitDegrees=45
  twistMotion=Free
 }
 {bone1Name=dyn_frontHair_01, bone2Name=base_frontHair, swing1LimitDegrees=5, swing2LimitDegrees=10, twistLimitDegrees=2}
 {bone1Name=dyn_frontHair_02, bone2Name=dyn_frontHair_01, swing1LimitDegrees=5, swing2LimitDegrees=10, twistLimitDegrees=2}
 {bone1Name=dyn_frontHair_03, bone2Name=dyn_frontHair_02, swing1LimitDegrees=5, swing2LimitDegrees=10, twistLimitDegrees=2}
]
