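# Post-physics AnimBP for the SleepyCloud female outfit's face accessory, adding secondary motion to its hair and ponytail.
# It copies the parent mesh pose, pitches the ponytail base, runs two neck-relative RigidBody nodes, then chains sprung AnimDynamics on left and right hair bones.
# Updates set gravity, front-end simulation space and alphas, plus hair toggles from backbling checks made on customization load, exempting non-long SleepyCloud backpacks.
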AnimBlueprint /BRCosmetics/Characters/Player/Female/Medium/Bodies/F_MED_SleepyCloud/Meshes/Parts/F_MED_SleepyCloud_FaceAcc_AnimBP
target_skeleton=&/F_MED_SleepyCloud_FaceAcc_Skeleton
graphs [
 AnimGraph {nodes [
   n1 = Anim Dynamics - Bone: dyn_hair_r_01_a!(
    SimulationSpace:AnimPhysSimSpaceType Enum
    ComponentPose=Local To Component(LocalPose=Use cached pose 'PonyTailEndCache'())
    Alpha=1
    ExternalForce="0,0,0"
   ) -> (Pose) {
    angularDampingOverride=1
    angularSpringConstant=6000
    bAngularSpring=true
    bOverrideAngularDamping=true
    bUseGravityOverride=true
    gravityScale=0.35
    simulationSpace=RootRelative
    alphaBoolBlend {blendOption=Linear}
    boundBone {boneName=dyn_hair_r_01_a}
    componentAppliedLinearAccClamp=<10000 10000 10000>
    componentLinearAccScale=<1 1 1>
    componentLinearVelScale=<1 1 1>
    gravityOverride {z=-980}
    physicsBodyDefinitions [{
      boundBone {boneName=dyn_hair_r_01_a}
      boxExtents=<12 30 12>
      constraintSetup {angularLimitsMax=<30 30 20>, angularLimitsMin=<-5 0 -3>, angularTarget=<1 0 0>}
      localJointOffset=<1 0 0>
     }]
   }
   n2 = Anim Dynamics - Bone: dyn_hair_r_02_a!(SimulationSpace:AnimPhysSimSpaceType Enum, ComponentPose=@n1, Alpha=1, ExternalForce="0,0,0") -> (Pose) {
    angularDampingOverride=1
    angularSpringConstant=6000
    bAngularSpring=true
    bOverrideAngularDamping=true
    bUseGravityOverride=true
    gravityScale=0.35
    simulationSpace=RootRelative
    alphaBoolBlend {blendOption=Linear}
    boundBone {boneName=dyn_hair_r_02_a}
    componentAppliedLinearAccClamp=<10000 10000 10000>
    componentLinearAccScale=<1 1 1>
    componentLinearVelScale=<1 1 1>
    gravityOverride {z=-980}
    physicsBodyDefinitions [{
      boundBone {boneName=dyn_hair_r_02_a}
      boxExtents=<12 30 12>
      constraintSetup {angularLimitsMax=<30 30 20>, angularLimitsMin=<-5 0 -3>, angularTarget=<1 0 0>}
      localJointOffset=<1 0 0>
     }]
   }
   n3 = Anim Dynamics - Bone: dyn_hair_r_03_a!(SimulationSpace:AnimPhysSimSpaceType Enum, ComponentPose=@n2, Alpha=1, ExternalForce="0,0,0") -> (Pose) {
    angularDampingOverride=1
    angularSpringConstant=6000
    bAngularSpring=true
    bOverrideAngularDamping=true
    bUseGravityOverride=true
    gravityScale=0.35
    simulationSpace=RootRelative
    alphaBoolBlend {blendOption=Linear}
    boundBone {boneName=dyn_hair_r_03_a}
    componentAppliedLinearAccClamp=<10000 10000 10000>
    componentLinearAccScale=<1 1 1>
    componentLinearVelScale=<1 1 1>
    gravityOverride {z=-980}
    physicsBodyDefinitions [{
      boundBone {boneName=dyn_hair_r_03_a}
      boxExtents=<12 30 12>
      constraintSetup {angularLimitsMax=<30 30 20>, angularLimitsMin=<-5 0 -3>, angularTarget=<1 0 0>}
      localJointOffset=<1 0 0>
     }]
   }
   n4 = Anim Dynamics - Bone: dyn_hair_r_04_a!(SimulationSpace:AnimPhysSimSpaceType Enum, ComponentPose=@n3, Alpha=1, ExternalForce="0,0,0") -> (Pose) {
    angularDampingOverride=1
    angularSpringConstant=6000
    bAngularSpring=true
    bOverrideAngularDamping=true
    bUseGravityOverride=true
    gravityScale=0.35
    simulationSpace=RootRelative
    alphaBoolBlend {blendOption=Linear}
    boundBone {boneName=dyn_hair_r_04_a}
    componentAppliedLinearAccClamp=<10000 10000 10000>
    componentLinearAccScale=<1 1 1>
    componentLinearVelScale=<1 1 1>
    gravityOverride {z=-980}
    physicsBodyDefinitions [{
      boundBone {boneName=dyn_hair_r_04_a}
      boxExtents=<12 30 12>
      constraintSetup {angularLimitsMax=<30 30 20>, angularLimitsMin=<-5 0 -3>, angularTarget=<1 0 0>}
      localJointOffset=<1 0 0>
     }]
   }
   n5 = Anim Dynamics - Bone: dyn_hair_r_01_b!(SimulationSpace:AnimPhysSimSpaceType Enum, ComponentPose=@n4, Alpha=0.6, ExternalForce="0,0,0") -> (Pose) {
    angularDampingOverride=1
    angularSpringConstant=6000
    bAngularSpring=true
    bOverrideAngularDamping=true
    bUseGravityOverride=true
    gravityScale=0.35
    simulationSpace=RootRelative
    alphaBoolBlend {blendOption=Linear}
    boundBone {boneName=dyn_hair_r_01_b}
    componentAppliedLinearAccClamp=<10000 10000 10000>
    componentLinearAccScale=<1 1 1>
    componentLinearVelScale=<1 1 1>
    gravityOverride {z=-980}
    physicsBodyDefinitions [{
      boundBone {boneName=dyn_hair_r_01_b}
      boxExtents=<12 30 12>
      constraintSetup {angularLimitsMax=<15 15 15>, angularLimitsMin=<-5 -15 -15>, angularTarget=<1 0 0>}
      localJointOffset=<1 0 0>
     }]
   }
   n6 = Anim Dynamics - Bone: dyn_hair_r_02_b!(SimulationSpace:AnimPhysSimSpaceType Enum, ComponentPose=@n5, Alpha=0.6, ExternalForce="0,0,0") -> (Pose) {
    angularDampingOverride=1
    angularSpringConstant=6000
    bAngularSpring=true
    bOverrideAngularDamping=true
    bUseGravityOverride=true
    gravityScale=0.35
    simulationSpace=RootRelative
    alphaBoolBlend {blendOption=Linear}
    boundBone {boneName=dyn_hair_r_02_b}
    componentAppliedLinearAccClamp=<10000 10000 10000>
    componentLinearAccScale=<1 1 1>
    componentLinearVelScale=<1 1 1>
    gravityOverride {z=-980}
    physicsBodyDefinitions [{
      boundBone {boneName=dyn_hair_r_02_b}
      boxExtents=<12 30 12>
      constraintSetup {angularLimitsMax=<15 15 15>, angularLimitsMin=<-5 -15 -15>, angularTarget=<1 0 0>}
      localJointOffset=<1 0 0>
     }]
   }
   n7 = Anim Dynamics - Bone: dyn_hair_l_01!(SimulationSpace:AnimPhysSimSpaceType Enum, ComponentPose=@n6, Alpha=1, ExternalForce="0,0,0") -> (Pose) {
    angularDampingOverride=1
    angularSpringConstant=6000
    bAngularSpring=true
    bOverrideAngularDamping=true
    bUseGravityOverride=true
    gravityScale=0.35
    simulationSpace=RootRelative
    alphaBoolBlend {blendOption=Linear}
    boundBone {boneName=dyn_hair_l_01}
    componentAppliedLinearAccClamp=<10000 10000 10000>
    componentLinearAccScale=<1 1 1>
    componentLinearVelScale=<1 1 1>
    gravityOverride {z=-980}
    physicsBodyDefinitions [{
      boundBone {boneName=dyn_hair_l_01}
      boxExtents=<12 30 12>
      constraintSetup {angularLimitsMax=<15 30 0>, angularLimitsMin=<-5 -20 -20>, angularTarget=<1 0 0>}
      localJointOffset=<2 0 0>
     }]
   }
   n8 = Anim Dynamics - Bone: dyn_hair_l_02!(SimulationSpace:AnimPhysSimSpaceType Enum, ComponentPose=@n7, Alpha=1, ExternalForce="0,0,0") -> (Pose) {
    angularDampingOverride=1
    angularSpringConstant=6000
    bAngularSpring=true
    bOverrideAngularDamping=true
    bUseGravityOverride=true
    gravityScale=0.35
    simulationSpace=RootRelative
    alphaBoolBlend {blendOption=Linear}
    boundBone {boneName=dyn_hair_l_02}
    componentAppliedLinearAccClamp=<10000 10000 10000>
    componentLinearAccScale=<1 1 1>
    componentLinearVelScale=<1 1 1>
    gravityOverride {z=-980}
    physicsBodyDefinitions [{
      boundBone {boneName=dyn_hair_l_02}
      boxExtents=<12 30 12>
      constraintSetup {angularLimitsMax=<15 0 0>, angularLimitsMin=<-5 -20 -20>, angularTarget=<1 0 0>}
      localJointOffset=<2 0 0>
     }]
   }
   Output Pose(Result=Component To Local(ComponentPose=@n8)) {layerGroup=None}
   n9 = "Transform (Modify) Bone - Bone: base_PonyTail"(
    ComponentPose=Local To Component(LocalPose=Copy Pose From Mesh() {bUseAttachedParent=true})
    Alpha=1
    Translation="0,0,0"
    Rotation="10,0,0"
    Scale="1,1,1"
   ) {
    rotationMode=BMM_Additive
    rotationSpace=BCS_BoneSpace
    alphaBoolBlend {blendOption=Linear}
    boneToModify {boneName=base_PonyTail}
    rotation {pitch=10}
   }
   n10 = RigidBody!(ComponentPose=@n9, Alpha=0, OverrideWorldGravity:Vector="0,0,-980", ExternalForce) -> (Pose) {
    bClampLinearTranslationLimitToRefPose=true
    bOverrideWorldGravity=true
    overridePhysicsAsset=&/F_MED_SleepyCloud_FaceAcc_Physics
    simulationSpace=BaseBoneSpace
    alphaBoolBlend {blendOption=Linear}
    baseBoneRef {boneName=neck_01}
    overrideWorldGravity {z=-980}
    simSpaceSettings {
     maxAngularAcceleration=8
     maxAngularVelocity=10
     maxLinearAcceleration=2000
     maxLinearVelocity=600
     velocityScaleZ=0.6
     worldAlpha=1
     externalLinearDragV {x=0.1}
    }
   }
   n11 = RigidBody(ComponentPose=@n10, OverrideWorldGravity="0,0,-980") {
    bClampLinearTranslationLimitToRefPose=true
    bOverrideWorldGravity=true
    overridePhysicsAsset=&/F_MED_SleepyCloud_FaceAcc_BP_Physics
    simulationSpace=BaseBoneSpace
    alphaBoolBlend {blendOption=Linear}
    baseBoneRef {boneName=neck_01}
    overrideWorldGravity {z=-980}
    simSpaceSettings {
     maxAngularAcceleration=8
     maxAngularVelocity=10
     maxLinearAcceleration=2000
     maxLinearVelocity=600
     velocityScaleZ=0.6
     worldAlpha=1
     externalLinearDragV {x=0.1}
    }
   }
   Save cached pose 'PonyTailEndCache'(Pose=Component To Local(ComponentPose=@n11)) {cacheName=PonyTailEndCache}
  ]}
 EventGraph {nodes [
   Event BlueprintUpdateAnimation()
   > n1 = Sequence!(execute) -> (then_0, then_1, then_2:Exec)
   > n2 = Gravity Override State Function DEPRECATED(CharacterPartAnimInstance=Self-Reference(), GameTimeInSeconds=GetGameTimeInSeconds(), GravityOverrideParams=Get Gravity Override Params())
   > Set GravityOverride(GravityOverride=@n2)
   > n3 = Branch(Condition=Get bIsInFrontEnd())
   > Set SimlationSpace(SimlationSpace:AnimPhysSimSpaceType Enum=World) -> (Output_Get:AnimPhysSimSpaceType Enum)
   > Set AlphaFE(AlphaFE:double=0.3) -> (Output_Get:double)
   > Set AlphaFE_Hair(AlphaFE_Hair:double=0.3) -> (Output_Get:double)
   Set SimlationSpace(execute=@n3.else)
   > Set AlphaFE(AlphaFE=1)
   > Set AlphaFE_Hair(AlphaFE_Hair=0.6)
   n4 = NOT Boolean(A=Get ISBPSleepyCloud() -> (ISBPSleepyCloud:bool))
   n5 = Get isWearingPack() -> (isWearingPack:bool)
   n6 = Branch(execute=@n1.then_1, Condition=OR Boolean(A=AND Boolean(A=@n5, B=@n4), B=Get isWearingCape() -> (isWearingCape:bool)))
   > Set EnabledDynHair(EnabledDynHair:double=1) -> (Output_Get:double)
   > Set DisableDynHair(DisableDynHair:double=0) -> (Output_Get:double)
   Set EnabledDynHair(execute=@n6.else, EnabledDynHair=0)
   > Set DisableDynHair(DisableDynHair=1)
   Event OnCharacterCustomizationLoaded()
   n7 = CheckForBackbling()
   > Set isWearingPack(isWearingPack:bool=OR Boolean!(A=@n7, B=@n7.bIsBackblingACape, C:bool=@n7.bIsBackblingLong) -> (ReturnValue)) -> (Output_Get:bool)
   > Set isWearingCape(isWearingCape=@n7.bIsBackblingLong)
   > n8 = Cast To FortPlayerPawn(Object=TryGetPawnOwner())
   n9 = GetSkeletalMeshForPartType(self=@n8, PartType=Backpack)
   > Set ISBPSleepyCloud(ISBPSleepyCloud:bool=MatchesWildcard(SourceString=GetDisplayName(Object=@n9), Wildcard=*SleepyCloud*)) -> (Output_Get:bool)
  ]}
]
variables [
 AlphaFE {type=number}
 AlphaFE_Hair {type=number}
 Backpack {type=boolean}
 DisableDynHair {type=number}
 EnabledDynHair {type=number}
 Gravity Override Params {type=GravityOverrideParamsStruct}
 GravityOverride {type=Vector}
 Is Wearing Pack {type=boolean}
 ISBPSleepyCloud {type=boolean}
 isWearingCape {type=boolean}
 isWearingPack {type=boolean}
 SimlationSpace {type=AnimPhysSimSpaceType}
]
